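# Physics asset for the StallionAviator two-handed pickaxe mesh, defining collision for the axe and a hanging tassel.
# The handle bone carries three capsules, and the four dyn_tassel bones use a small capsule followed by three boxes.
# Angular-limited constraints chain the tassel bones in sequence from the handle, with almost no twist and tighter swing on the lower two links.
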PhysicsAsset /BRCosmetics/Weapons/FORT_Melee/Pickaxe_StallionAviator/Meshes/StallionAviator_Axe_Physics
bodies {
 Melee_TwoHanded_Handle [
  {shapeType=Capsule, length=34.9151, radius=1.4098, center=<0.0044 -0.1519 -0.5021>, rotation=<rot -0.3505 -89.9997 0.012>}
  {
   shapeType=Capsule
   length=7.8298
   radius=15.4591
   shapeName=Melee_TwoHanded_Handle_capsule
   center=<17.1727 2.2791 -18.8175>
   rotation=<rot 180 0 0>
  }
  {shapeType=Capsule, length=1.8908, radius=2.4233, shapeName=Melee_TwoHanded_Handle_capsule, center=<0 0 -21.5726>}
 ]
 dyn_tassel_01 [{shapeType=Capsule, length=1.0516, radius=0.5, shapeName=dyn_tassel_01_capsule, rotation=<rot 0 0 90>}]
 dyn_tassel_02 [{shapeType=Box, extentX=2.1686, extentY=2.1647, extentZ=0.8099, shapeName=dyn_tassel_02_box}]
 dyn_tassel_03 [{shapeType=Box, extentX=3.349, extentY=2.1647, extentZ=0.9234, shapeName=dyn_tassel_03_box, center=<0.4181 0 0>}]
 dyn_tassel_04 [{
   shapeType=Box
   extentX=1.9688
   extentY=3.4286
   extentZ=2.2857
   shapeName=dyn_tassel_04_box
   center=<1.1743 0.2093 0>
   rotation=<rot 0 100.1042 0>
  }]
}
constraints [
 {bone1Name=dyn_tassel_01, bone2Name=Melee_TwoHanded_Handle, swing1LimitDegrees=45, swing2LimitDegrees=45, twistLimitDegrees=2}
 {bone1Name=dyn_tassel_02, bone2Name=dyn_tassel_01, swing1LimitDegrees=45, swing2LimitDegrees=45, twistLimitDegrees=1}
 {bone1Name=dyn_tassel_03, bone2Name=dyn_tassel_02, swing1LimitDegrees=15, swing2LimitDegrees=25, twistLimitDegrees=1}
 {bone1Name=dyn_tassel_04, bone2Name=dyn_tassel_03, swing1LimitDegrees=15, swing2LimitDegrees=44.48, twistLimitDegrees=1}
]
