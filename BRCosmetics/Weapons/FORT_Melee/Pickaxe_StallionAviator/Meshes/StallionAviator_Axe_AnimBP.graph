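# Animation Blueprint for the StallionAviator pickaxe that adds secondary motion over its base weapon pose, including on its tassel bone chain.
# When bRigidBodyEnabled is true, the base pose passes through DefaultSlot into a RigidBody simulation relative to the handle bone.
# Otherwise, an offset tassel root feeds a four-bone Trail controller with rotation limits and a handle-driven planar limit.
# The input-pose cache is unconnected to the output, and the event graph holds no logic.
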AnimBlueprint /BRCosmetics/Weapons/FORT_Melee/Pickaxe_StallionAviator/Meshes/StallionAviator_Axe_AnimBP
target_skeleton=&/StallionAviator_Axe_Skeleton
graphs [
 AnimGraph {nodes [
   n1 = Sequence Player 'StallionAviator_base_pose_Weapon'() {sequence=/BRCosmetics/Animation/Game/MainPlayer/Combat/Melee/OneHanded/StallionAviator/StallionAviator_base_pose_Weapon}
   n2 = RigidBody(ComponentPose=Local To Component(LocalPose=Slot 'DefaultSlot'(Source=@n1))) {
    bClampLinearTranslationLimitToRefPose=true
    bOverrideWorldGravity=true
    overridePhysicsAsset=&/StallionAviator_Axe_Physics
    simulationSpace=BaseBoneSpace
    alphaBoolBlend {blendOption=Linear}
    baseBoneRef {boneName=Melee_TwoHanded_Handle}
    componentAppliedLinearAccClamp=<2000 2000 2000>
    componentLinearAccScale=<0.5 0.4 0.4>
    componentLinearVelScale=<0.5 0.4 0.4>
    overrideWorldGravity {z=-980}
   }
   n3 = "Transform (Modify) Bone - Bone: dyn_tassel_01"(
    ComponentPose=Local To Component(LocalPose=Sequence Player 'StallionAviator_base_pose_Weapon'() {sequence=&StallionAviator_base_pose_Weapon})
    Alpha=1
    Translation="0,1.5,0.2"
    Rotation="0,90,0"
    Scale="1,1,1"
   ) {
    rotationMode=BMM_Additive
    rotationSpace=BCS_BoneSpace
    translationMode=BMM_Additive
    translationSpace=BCS_BoneSpace
    alphaBoolBlend {blendOption=Linear}
    boneToModify {boneName=dyn_tassel_01}
    rotation {yaw=90}
    translation {y=1.5, z=0.2}
   }
   n4 = Trail controller - Bone: dyn_tassel_04(ComponentPose=@n3, Alpha=1) {
    bUsePlanarLimit=true
    chainLength=4
    debugLifeTime=0.1
    maxDeltaTime=0.0333
    stretchLimit=1
    baseJoint {boneName=b_AR_Root}
    fakeVelocity {z=50}
    planarLimits [{drivingBone {boneName=Melee_TwoHanded_Handle}, planeTransform {location=<-3 0 0>, rotation=<rot -90 180 0>, scale=<1 1 1>}}]
    rotationLimits [
     {limitMax=<20 20 20>, limitMin=<-20 -20 -20>}
     {limitMax=<20 20 90>, limitMin=<-90 -90 -90>}
     {limitMax=<90 90 90>, limitMin=<-90 -90 -90>}
     {limitMax=<180 180 180>}
    ]
    rotationOffsets [<0 0 0>, <0 0 0>, <0 0 0>, <0 0 0>]
    trailBone {boneName=dyn_tassel_04}
    trailRelaxationSpeed {editorCurveData {keys [<at 0 -4.2363 RCIM_Linear>,<at 0.1833 -10.4238 RCIM_Linear>,<at 0.4555 -16.0957 RCIM_Linear>,<at 1 -20.7363 RCIM_Linear>]}}
   }
   Output Pose(Result=Blend Poses by bool(BlendPose_0=Component To Local(ComponentPose=@n2), BlendPose_1=Component To Local(ComponentPose=@n4), bActiveValue=Get bRigidBodyEnabled()) {blendType=Linear}) {layerGroup=None}
   AnimGraphNode_SubInput_0 = Input Pose()
   Save cached pose 'SavedPose11465'(Pose=@AnimGraphNode_SubInput_0) {cacheName=SavedPose11465}
   Use cached pose 'SavedPose11465'()
   Use cached pose 'SavedPose11465'()
  ]}
 EventGraph {nodes [
   TryGetPawnOwner()
   Event BlueprintUpdateAnimation()
  ]}
]
variables []
